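; HUD progress-bar view model for the Void Run effect granted by the Fire Petal Boon.
; On initialize with a valid Pawn, it chains four gameplay tag query wait tasks for Void Run and boon equip state, then sets a radial override material.
; Void Run toggles the bar between full and force-visible, or empty without forced visibility.
; Boon changes set whether the bar is enabled and always reset it to empty.
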(event EventInitialize
  (params (Pawn :type Pawn%20Object%20Reference))
  (|CallParentFunction Pawn)
  (Utilities|IsValid Pawn
    (:Is%20Valid
      ;[1 0.9905 0.4115 0.5] Listening to when the player starts the Void Run effect
      (Ability|Tasks|WaitGameplayTagQueryOnActor Pawn (Variables|Default|GetVoidRunEnergyTagQuery)
        (:then
          (Ability|Tasks|WaitGameplayTagQueryOnActor Pawn (Variables|Default|GetVoidRunEnergyTagQuery) "WhenFalse"
            (:then
              ;[1 0.5512 0.7143 0.5] Listening to when the player starts equips the Petal Boon
              (Ability|Tasks|WaitGameplayTagQueryOnActor Pawn (Variables|Default|GetPetalBoonEnergyTagQuery)
                (:then
                  (Ability|Tasks|WaitGameplayTagQueryOnActor Pawn (Variables|Default|GetPetalBoonEnergyTagQuery) "WhenFalse"
                    (:then
                      (CallFunction|SetProgressBarOverrideMaterial "/FirePetalBoonsGameplay/UI/Materials/MI_UI_VoidRun_Radial.MI_UI_VoidRun_Radial"))
                    (:Triggered
                      (CallFunction|TogglePetalBoonEquipped false))))
                (:Triggered
                  (CallFunction|TogglePetalBoonEquipped true))))
            (:Triggered
              (CallFunction|ToggleVoidRunEnergyBar false))))
        (:Triggered
          (CallFunction|ToggleVoidRunEnergyBar true))))))

(fn TogglePetalBoonEquipped
  (params (IsEquipped :type Boolean))
  (Variables|GenericProgressBarViewModel|SetbEnabled IsEquipped)
  (CallFunction|ToggleVoidRunEnergyBar false))

(fn ToggleVoidRunEnergyBar
  (params (bIsActive :type Boolean))
  (if bIsActive
    (Variables|GenericProgressBarViewModel|SetProgressValue 1.0)
    (Variables|GenericProgressBarViewModel|SetbForceVisible true)
    (else
      (Variables|GenericProgressBarViewModel|SetProgressValue 0.0)
      (Variables|GenericProgressBarViewModel|SetbForceVisible false))))
